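; Base Blueprint for PlotTwist lightsaber pickups, adding variant-colored rarity effects on top of its parent pickup logic.
; Construction passes the VariantFX index to the Niagara pickup system.
; Shortly after BeginPlay, that index selects a ColorVariant color that tints the ambient emitter and rare-pickup mesh, which also gets a sparkle material.
; On pickup, it spawns a tinted rarity burst and flourish sound, then hides BG and the spotlight.
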(event EventBeginPlay
  (bind _variantfx (Variables|Default|GetVariantFX))
  (bind _colorvariant (Variables|Default|GetColorVariant))
  (bind _output (Utilities|Array|Get(acopy) _colorvariant _variantfx))
  (bind _returnvalue (Math|Conversions|ToVector(LinearColor) _output))
  (bind _rarepickupfxmesh (Variables|Default|GetrarePickupFXMesh))
  (|CallParentFunction)
  (Utilities|FlowControl|Delay 0.2) ; Slight delay to allow parent Pickup BP to get set up
  (Rendering|Material|SetMaterial (Variables|Default|GetrarePickupFXMesh) 2 (Variables|Default|GetRare%20Pickup%20Sparkle%20Material))
  (Niagara|SetNiagaraVariable(LinearColor) (Variables|Default|GetrareAmbientEmitter) "RarityColor" _output)
  (Rendering|Material|SetVectorParameterValueonMaterials _rarepickupfxmesh "Color_Sparkle" _returnvalue)
  (Rendering|Material|SetVectorParameterValueonMaterials _rarepickupfxmesh "BaseColor" _returnvalue))

(event Collision|EventActorBeginOverlap
  (params (OtherActor :type Actor%20Object%20Reference)))

(event EventTick
  (params (DeltaSeconds :type Float%20%28single-precision%29))
  (|CallParentFunction DeltaSeconds) ; This node is disabled and will not be called.\nDrag off pins to build functionality.
)

(event EventOnPickedUp
  (params (PickupTarget :type Fort%20Pawn%20Object%20Reference))
  (bind _returnvalue (Niagara|SpawnSystematLocation "/Game/Effects/Fort_Effects/Gameplay/Pickups/Effects/NS_Pickup_Rarity.NS_Pickup_Rarity" (Transformation|GetWorldLocation (Variables|Transformation|GetRootComponent))))
  (Niagara|SetNiagaraVariable(LinearColor) _returnvalue "RarityColor" (Utilities|Array|Get(acopy) (Variables|Default|GetColorVariant) (Variables|Default|GetVariantFX)))
  (Audio|SpawnSoundatLocation "/Game/Sounds/Fort_GamePlay_Sounds/Loot/Pickup_Weapon_Flourish_Cue.Pickup_Weapon_Flourish_Cue" (Transformation|GetWorldLocation (Variables|Transformation|GetRootComponent)))
  (bind _bg (Variables|Default|GetBG)
    (:then
      (Development|SetHiddeninGame _bg true true))
    (:else
      _))
  (bind _spotlightcomp (Variables|Default|GetspotlightComp)
    (:then
      (Rendering|SetVisibility _spotlightcomp false false))
    (:else)))

(fn ConstructionScript
  (|CallParentFunction)
  (Niagara|SetNiagaraVariable(Int32) (Variables|Default|GetNS_PlotTwist_Pickup_Base) "ColorVariant" (Variables|Default|GetVariantFX)))
