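# Translucent unlit Niagara sprite material that draws an animated noisy ring, apparently for the PlotTwist lightsaber pickup.
# Builds polar UVs from angle and squared radius, scrolls them, and distorts the radial axis with panning soft noise.
# Directional noise sampled there is masked by a sine-shaped exponential radial gradient, offset by a dynamic parameter.
# The mask scales particle alpha; its square blends particle color toward a grayscale second color taken only from the R parameter.
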Material /PlotTwist/Items/LightSaber/Pickup/Materials/M_PlotTwist_Pickup_Ring
graph {
 properties {bUsedWithNiagaraSprites=true, blendMode=BLEND_Translucent, pixelDepthOffsetMode=PDOM_Legacy, shadingModel=MSM_Unlit, usedShadingModels=Unlit}
 nodes [
  n1 = ParticleColor()
  n2 = DynamicParameter() {defaultValue=<rgba 1 0.6808 1 0.6615>, paramNames [UV Random Pos, 2nd color R, 2nd color G, 2nd color B]}
  n3 = DynamicParameter() {defaultValue=<rgba 1 0.6808 1 0.6615>, paramNames [UV Random Pos, 2nd color R, 2nd color G, 2nd color B]}
  n4 = Subtract(A=TextureCoordinate()) {constB=0.5}
  n5 = ComponentMask(None=@n4) {g=true}
  n6 = ComponentMask(None=@n4) {r=true}
  n7 = Arctangent2!(Y=@n5, X=@n6)
  n8 = MaterialFunctionCall() {materialFunction=&Pi}
  n9 = Multiply(A=ScalarParameter() {defaultValue=1, parameterName=Y_Tiling}, B=Divide(A=Add(A=@n7, B=@n8.Result), B=Multiply(A=@n8.Result) {constB=2}))
  n10 = Time()
  n11 = Multiply(A=ScalarParameter() {defaultValue=0.02, parameterName=Y_Scroll}, B=@n10)
  n12 = Multiply(A=@n10, B=ScalarParameter() {defaultValue=-0.05, parameterName=X_Scroll})
  n13 = Multiply(A=@n6, B=@n6)
  n14 = Multiply(A=@n5, B=@n5)
  n15 = DynamicParameter() {defaultValue=<rgba 1 0.6808 1 0.6615>, paramNames [UV Random Pos, 2nd color R, 2nd color G, 2nd color B]}
  n16 = Add(
   A=TextureCoordinate() {uTiling=0.5, vTiling=0.5}
   B=Multiply(A=Time(), B=ScalarParameter() {defaultValue=0.19, parameterName=Distortion Speed})
  )
  n17 = ComponentMask(None=TextureSample(UVs=Add(A=@n15.UV Random Pos, B=@n16)) {texture=/Game/Effects/Fort_Effects/Emotes/UltraLight/Texture/T_SoftNoise_BC7}.RGB) {r=true}
  n18 = Add(
   A=Multiply(A=ScalarParameter() {defaultValue=4, parameterName=X_Tiling}, B=Add(A=@n13, B=@n14))
   B=Multiply(A=Subtract(A=@n17) {constB=0.5}, B=ScalarParameter() {defaultValue=0.1, parameterName=Distortion Scale}) {constB=0.2}
  )
  n19 = Add(A=@n3.UV Random Pos, B=AppendVector(A=Add(A=@n9, B=@n11), B=Add(A=@n12, B=@n18)))
  n20 = TextureSample!(
   UVs=@n19
   Tex
   Level=MaterialFunctionCall!(UVs=ConstantBiasScale(None=@n19), Texture Size=Constant() {r=1}) {materialFunction=/Engine/Functions/Engine_MaterialFunctions01/Texturing/ComputeMipLevel}.MipLevel
   Apply View MipBias
  ) {
   mipValueMode=TMVM_MipLevel
   samplerType=SAMPLERTYPE_LinearGrayscale
   texture=/Game/Effects/Fort_Effects/Emotes/UltraLight/Texture/T_DirectionalNoise_Greyscale
  }
  n21 = ScalarParameter() {defaultValue=0.47, parameterName=Ring Radius}
  n22 = ScalarParameter() {defaultValue=1.5, parameterName=Ring Density}
  n23 = Multiply(
   A=@n20.RGB
   B=Sine(None=MaterialFunctionCall(Radius=@n21, Density=@n22) {materialFunction=&RadialGradientExponential}.RadialGradientExponential)
  )
  LinearInterpolate(
   A=@n1.RGB
   B=MaterialFunctionCall(X=@n2.2nd color R, Y=@n2.2nd color R, Z=@n2.2nd color R) {materialFunction=&MakeFloat3}.Result
   Alpha=Multiply(A=@n23, B=@n23)
  )
  Multiply(A=@n1.A, B=@n23)
  VectorParameter() {parameterName=Second Color, defaultValue=<rgba 0.6808 1 0.6615 1>}
 ]
}
