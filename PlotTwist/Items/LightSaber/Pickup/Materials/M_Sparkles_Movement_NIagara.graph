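# Unlit translucent, two-sided Niagara sprite and mesh particle material for sparkles, filed under the lightsaber pickup.
# Colour blends particle colour toward a Dynamic Parameter colour by a flare-texture channel, though the single-colour parameter defaults to plain particle colour.
# A mask, switchable from texture to procedural SphereMask, multiplies particle alpha and, through Is Additive, optionally the colour entering MF_FX_EmissiveAdjustments.
# The graph contains no movement logic despite the name.
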Material /PlotTwist/Items/LightSaber/Pickup/Materials/M_Sparkles_Movement_NIagara
graph {
 properties {
  bCastRayTracedShadows=false
  bEnableResponsiveAA=true
  bUsedWithNiagaraMeshParticles=true
  bUsedWithNiagaraSprites=true
  blendMode=BLEND_Translucent
  opacityMaskClipValue=0.01
  pixelDepthOffsetMode=PDOM_Legacy
  shadingModel=MSM_Unlit
  twoSided=true
  usedShadingModels=Unlit
 }
 nodes [
  n1 = ParticleColor()
  n2 = DynamicParameter() {paramNames [Color 2 R, Color 2 G, Color 2 B, Hueshift]}
  n3 = ChannelMaskParameter(None=TextureSampleParameter2D() {group=Textures - Opacity, parameterName=Opacity Mask Texture, texture=/Game/Effects/Fort_Effects/Textures/Flares/T_GlowSar_6pt}.RGB) {group=Textures - Opacity, maskChannel=Blue, parameterName=Mask Texture Channel Mask, defaultValue=<rgba 0 0 1 0>}
  n4 = Clamp(None=LinearInterpolate(
    A=LinearInterpolate(A=@n1.RGB, B=@n2.RGB, Alpha=@n3.R)
    B=ParticleColor().RGB
    Alpha=ScalarParameter() {defaultValue=1, group=Color, parameterName=Use Single Color Parameter}
   )) {maxDefault=4}
  n5 = ChannelMaskParameter(None=TextureSampleParameter2D() {group=Textures - Opacity, parameterName=Opacity Mask Texture, texture=&T_GlowSar_6pt}.RGB) {group=Textures - Opacity, maskChannel=Blue, parameterName=Mask Texture Channel Mask, defaultValue=<rgba 0 0 1 0>}
  n6 = LinearInterpolate(
   A=@n5.R
   B=SphereMask(
    A=TextureCoordinate()
    B=Constant2Vector() {g=0.5, r=0.5}
    Radius=ScalarParameter() {defaultValue=0.45, parameterName=Radius}
    Hardness=ScalarParameter() {parameterName=Hardness}
   )
   Alpha=ScalarParameter() {parameterName=Use Simple Opacity}
  )
  MaterialFunctionCall(Emissive=LinearInterpolate(A=@n4, B=Multiply(A=@n4, B=@n6), Alpha=ScalarParameter() {group=Overall Settings, parameterName=Is Additive})) {materialFunction=&MF_FX_EmissiveAdjustments}
  Multiply(A=ParticleColor().A, B=@n6)
 ]
}
