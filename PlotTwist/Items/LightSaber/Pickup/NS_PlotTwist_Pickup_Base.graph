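# Base Niagara glow effect for PlotTwist lightsaber pickups, with a User.ColorVariant index into five-color palettes.
# SaberHandleGlow bursts sprites along a vertical line, pulses them by sine and blends amber toward the variant color.
# HangingParticulates spawns short-lived curl-noise motes that stay amber because their variant lerp is zero.
# Time-of-day scalars adjust alpha inputs; the RingGlow emitter is disabled.
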NiagaraSystem /PlotTwist/Items/LightSaber/Pickup/NS_PlotTwist_Pickup_Base
dependencies {
 usedDataInterfaces [&NiagaraDataInterfaceArrayColor, &NiagaraDataInterfaceCurve]
 usedDynamicInputs []
 usedModules [&SystemState]
}
emitters {
 HangingParticulates {
  emitterUpdateScript [
   &EmitterState(Life Cycle Mode=::NewEnumerator1, Loop Duration=2, UseLoopDelay=false)
   &SpawnRate(Spawn Probability=0.5, SpawnRate=20, Use Spawn Probability=true)
  ]
  particleSpawnScript [
   &InitializeParticle(
    Color=&HueShiftLinearColor(
     Hue=&RandomRangeFloat(Minimum=0, Maximum=0.01)
     Saturation=1
     Value=1
     Linear Color=&MultiplyLinearColor(
      Linear Color=&LerpLinearColors(
       StartColor=<rgba 0.775 0.3185 0.004 0.25>
       EndColor=&SelectLinearColorFromArray(
        Array Sampling Mode=::NewEnumerator1
        Direct Array Index=$User.ColorVariant
        Color Selection Array=DataInterface() {propertyValues {ColorData [<rgba 0.02 0.1691 1 1>, <rgba 0.01 0.5 0.01 1>, <rgba 0.3467 0.02 1 1>, <rgba 0.5 0.5 0.5 1>, <rgba 1 0 0 1>]}}
        Random Seed=?
       )
       LerpFactor=0
      )
      Scale Factor=&MakeFloat4FromVectorAndFloat(
       "Vector (RGB)"=&VectorFromFloat(Value=1)
       "Float (Alpha)"=&FNTimeOfDayScalar(Night Scale=0.9, Morning Scale=0.5, Day Scale=1.5, Evening Scale=0.6)
      )
     )
    )
    Lifetime Max=1.8
    Lifetime Min=0.5
    Lifetime Mode=::NewEnumerator1
    Sprite Rotation Mode=::NewEnumerator2
    Sprite Size Mode=::NewEnumerator3
    Sprite UV Mode=::NewEnumerator3
    Uniform Sprite Size=&RandomRangeFloat(Minimum=35, Maximum=35)
    UsePositionOffset=false
    Write Material Random=true
   )
   &ShapeLocation(MaskSpawnGroup=false, Override Seed=false, Sphere Radius=35, UseOwnerScale=true)
  ]
  particleUpdateScript [
   &ParticleState()
   &ScaleSpriteSize(Uniform Curve Sprite Scale=DataInterface() {propertyValues {Curve {
       defaultValue=0
       keys [
        {arriveTangent=17.8316, interpMode=RCIM_Cubic, leaveTangent=17.8316, tangentMode=RCTM_User, time=0, value=0}
        {arriveTangent=-0.7251, arriveTangentWeight=0.0252, interpMode=RCIM_Cubic, leaveTangent=-0.7251, tangentMode=RCTM_User, time=0.0756, value=1}
        {arriveTangent=-1.3569, arriveTangentWeight=0.0415, interpMode=RCIM_Cubic, leaveTangent=-1.3569, tangentMode=RCTM_User, time=0.2, value=0.6}
        <at 1 0 RCIM_Cubic>
       ]
      }}}) {module=ScaleSpriteSize001}
   &ScaleColor(
    Scale Alpha=&FloatFromCurve(
     FloatCurve=DataInterface() {propertyValues {Curve {
        defaultValue=0
        keys [
         {arriveTangent=17.8316, interpMode=RCIM_Cubic, leaveTangent=17.8316, tangentMode=RCTM_User, time=0, value=0}
         {arriveTangent=-0.7251, arriveTangentWeight=0.0252, interpMode=RCIM_Cubic, leaveTangent=-0.7251, tangentMode=RCTM_User, time=0.0756, value=1}
         {arriveTangent=-1.3569, arriveTangentWeight=0.0415, interpMode=RCIM_Cubic, leaveTangent=-1.3569, tangentMode=RCTM_User, time=0.2, value=0.6}
         <at 1 0 RCIM_Cubic>
        ]
       }}}
     Scale Curve=&RandomRangeFloat(Minimum=0.5, Maximum=1)
    )
    Scale RGB=&VectorFromFloat(Value=2)
    ScaleA=false
    ScaleRGB=true
   )
   &CurlNoiseForce(
    Bias Noise=true
    Bias Noise Field:Vector3f=&RandomRangeVector(Minimum=<-1 -1 -0.1>, Maximum=<1 1 0.1>)
    Noise Frequency=120
    Noise Quality / Cost=::NewEnumerator0
    Noise Strength=15
    Pan Noise=false
   )
   &Drag(Use Linear Drag=true, Use Rotational Drag=true)
   &SolveForcesAndVelocity(Clamp Velocity=false, Limit Acceleration=false)
  ]
  rendererClasses [&NiagaraSpriteRendererProperties]
  renderers [{rendererClass=&NiagaraSpriteRendererProperties}]
 }
 RingGlow {
  bEnabled=false
  emitterUpdateScript [
   &EmitterState()
   &SpawnBurst_Instantaneous(Use Spawn Probability=false, UseLoopCountLimit=true)
  ]
  particleSpawnScript [
   &InitializeParticle(
    Color=&MultiplyLinearColor(
     Linear Color=&SelectLinearColorFromArray(
      Array Sampling Mode=::NewEnumerator1
      Direct Array Index=$User.ColorVariant
      Color Selection Array=DataInterface() {propertyValues {ColorData [<rgba 0.02 0.1691 1 1>, <rgba 0.01 0.5 0.01 1>, <rgba 0.3467 0.02 1 1>, <rgba 0.5 0.5 0.5 1>, <rgba 1 0 0 1>]}}
      Random Seed=?
     )
     Scale Factor=&MakeFloat4FromVectorAndFloat(
      "Vector (RGB)"=&VectorFromFloat(Value=3)
      "Float (Alpha)"=&FNTimeOfDayScalar(Night Scale=0.9, Morning Scale=0.5, Day Scale=1.5, Evening Scale=0.6)
     )
    )
    Sprite Rotation Mode=::NewEnumerator2
    Sprite Size Mode=::NewEnumerator3
    Uniform Sprite Size=82.5
    UsePositionOffset=false
    Write Material Random=true
   )
   &DynamicMaterialParameters(
    Index 0 Vector 4:Vector4f=&self:ScratchDynamicInput(
     Default:Vector3f=&MakeVectorFromLinearColorRGB(Color=&SelectLinearColorFromArray(
       Array Sampling Mode=::NewEnumerator1
       Direct Array Index=$User.ColorVariant
       Color Selection Array=DataInterface() {propertyValues {ColorData [<rgba 1.0162 5.6045 25 1>, <rgba 1.5 3 1.5 1>, <rgba 3.6667 0.5 10 1>, <rgba 10 10 10 1>, <rgba 7 1.0957 1.0957 1>]}}
       Random Seed=?
      ))
     float:NiagaraFloat=&RandomRangeFloat(Minimum=0, Maximum=1)
    )
    Write Parameter Index 0=::NewEnumerator4
   )
   &CameraOffset(Camera Offset Amount=-50)
  ]
  particleUpdateScript [
   &ParticleState(Kill Particles When Lifetime Has Elapsed=false, Let Infinitely Lived Particles Die When Emitter Deactivates=true)
   &ScaleColor(Scale Alpha=0.6, ScaleA=true, ScaleRGB=true)
  ]
  rendererClasses [&NiagaraSpriteRendererProperties]
  renderers [{rendererClass=&NiagaraSpriteRendererProperties}]
 }
 SaberHandleGlow {
  emitterSpawnScript [/SetVariables(Emitter.NumLine=10) {isSetParameters=true}]
  emitterUpdateScript [
   &EmitterState()
   &SpawnBurst_Instantaneous(Spawn Count=$Emitter.NumLine, Use Spawn Probability=true, UseLoopCountLimit=true)
  ]
  particleSpawnScript [
   &PartitionParticles(Number Of Partitions=1, Particles Per Partition:NiagaraInt32=$Emitter.NumLine, Partition Mode:ENiagara_PartitionMode_v2=::NewEnumerator1)
   &InitializeParticle(
    Color=&LerpLinearColors(
     StartColor=<rgba 1 0.3866 0 0.255>
     EndColor=&MultiplyLinearColor(
      Linear Color=&SelectLinearColorFromArray(
       Array Sampling Mode=::NewEnumerator1
       Direct Array Index=$User.ColorVariant
       Color Selection Array=DataInterface() {propertyValues {ColorData [<rgba 0.02 0.1691 1 1>, <rgba 0.008 0.4 0.008 1>, <rgba 0.3467 0.02 1 1>, <rgba 0.5 0.5 0.5 1>, <rgba 1 0 0 1>]}}
       Random Seed=?
      )
      Scale Factor=&MakeFloat4FromVectorAndFloat(
       "Vector (RGB)"=&VectorFromFloat(Value=1)
       "Float (Alpha)"=&Multiply_Float(
        A=&FNTimeOfDayScalar(Night Scale=0.9, Morning Scale=0.5, Day Scale=1.5, Evening Scale=0.6)
        B=0.35
       )
      )
     )
     LerpFactor=0
    )
    Position Offset=&Lerp_Vector(A=<0 0 32>, B=<0 0 -40>, Alpha=$Particles.PartitionParticles.Normalized Index In Partition)
    Sprite Size Mode=::NewEnumerator3
    Sprite UV Mode=::NewEnumerator3
    Uniform Sprite Size=57
    UsePositionOffset=true
    Write Material Random=true
   )
   &CameraOffset(Camera Offset Amount=-3)
  ]
  particleUpdateScript [
   &ParticleState(Kill Particles When Lifetime Has Elapsed=false, Let Infinitely Lived Particles Die When Emitter Deactivates=true)
   &ScaleColor(
    Scale Alpha=3
    Scale RGB=&Multiply_VectorByFloat(Vector=<1 1 1>, Float=&Sine(Normalized Angle=$Emitter.Age, Period=3, Scale=0.5, Bias=1.2))
    ScaleA=true
    ScaleRGB=true
   )
   &ScaleColor(
    Scale Alpha=&FNTimeOfDayScalar(Night Scale=0.25, Morning Scale=0.3, Day Scale=0.5, Evening Scale=0.3)
    Scale RGB=&VectorFromFloat(Value=1)
    ScaleA=true
    ScaleRGB=true
   ) {module=ScaleColor001}
   /SetVariables(
    Particles.Linear Color=&LerpLinearColors(
     StartColor=<rgba 1 0.3866 0 0.255>
     EndColor=&MultiplyLinearColor(
      Linear Color=&SelectLinearColorFromArray(
       Array Sampling Mode=::NewEnumerator1
       Direct Array Index=$User.ColorVariant
       Color Selection Array=DataInterface() {propertyValues {ColorData [<rgba 0.02 0.1691 1 1>, <rgba 0.008 0.4 0.008 1>, <rgba 0.3467 0.02 1 1>, <rgba 0.5 0.5 0.5 1>, <rgba 3 0 0 1>]}}
       Random Seed=?
      )
      Scale Factor=&MakeFloat4FromVectorAndFloat(
       "Vector (RGB)"=&VectorFromFloat(Value=1)
       "Float (Alpha)"=&Multiply_Float(A=&FNTimeOfDayScalar(Night Scale=0.9, Morning Scale=0.5, Day Scale=1.5, Evening Scale=0.6), B=5)
      )
     )
     LerpFactor=0.75
    )
    Particles.Position2:NiagaraPosition=&self:ScratchDynamicInput_01(Position:NiagaraPosition=$Particles.Position, float:NiagaraFloat=0.9)
    Particles.SpriteSize2=&Multiply_Vector2D(A=&Multiply_Vector2D_ByFloat(Float=0.7, Vector2D=<0.6 0.4>), B=$Particles.SpriteSize)
   ) {isSetParameters=true}
  ]
  rendererClasses [&NiagaraSpriteRendererProperties]
  renderers [{rendererClass=&NiagaraSpriteRendererProperties}]
 }
}
summary {
 systemName=NS_PlotTwist_Pickup_Base
 emitters [
  {bEnabled=false, emitterName=RingGlow, rendererClasses [&NiagaraSpriteRendererProperties]}
  {emitterName=SaberHandleGlow, rendererClasses [&NiagaraSpriteRendererProperties]}
  {emitterName=HangingParticulates, rendererClasses [&NiagaraSpriteRendererProperties]}
 ]
 userVariables [User.ColorVariant {type=NiagaraInt32, defaultValue=2}]
}
